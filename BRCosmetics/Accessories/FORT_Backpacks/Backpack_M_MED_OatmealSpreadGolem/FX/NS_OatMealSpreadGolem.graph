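# Niagara system for the OatmealSpreadGolem back bling that draws animated spotlights, with a User.KillCount input for elimination reactions.
# One emitter bursts 20 persistent particles at array-defined positions, split into top and bottom groups.
# Sine/cosine offsets sweep each outward-leaning beam's alignment, while four sprite renderers provide layering.
# System.ElimAlpha raises speed, sweep, size and pulsing brightness, shifting warm white toward blue, white and red.
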NiagaraSystem /BRCosmetics/Accessories/FORT_Backpacks/Backpack_M_MED_OatmealSpreadGolem/FX/NS_OatMealSpreadGolem
dependencies {
 usedDataInterfaces [
  &NiagaraDataInterfaceArrayFloat3
  &NiagaraDataInterfaceCurve
  &NiagaraDataInterfaceArrayColor
  &NiagaraDataInterfaceSocketReader
  &NiagaraDataInterfaceArrayFloat
 ]
 usedDynamicInputs [&FloatFromCurve, &MakeCustomFloatFromBool]
 usedModules [
  &self:SystemSpawnScript.SystemScriptSource.SystemScriptGraph.NiagaraNodeAssignment_0.SetVariables_#1
  &SystemState
  &self:SystemSpawnScript.SystemScriptSource.SystemScriptGraph.NiagaraNodeAssignment_2.SetVariables_#2
  /Game/Effects/Niagara/Modules/Logic/UpdateValueStateChange
  &Timeline
  &self:SystemSpawnScript.SystemScriptSource.SystemScriptGraph.NiagaraNodeAssignment_1.SetVariables_#3
  &self:SystemSpawnScript.SystemScriptSource.SystemScriptGraph.NiagaraNodeAssignment_3.SetVariables_#4
 ]
}
emitters {NE_Spotlights {
  emitterSpawnScript [/SetVariables() {isSetParameters=true}]
  emitterUpdateScript [
   &EmitterState(
    Inactive Response=::NewEnumerator1
    Life Cycle Mode=::NewEnumerator1
    Loop Behavior=::NewEnumerator1
    Loop Duration Mode=::NewEnumerator1
    UseLoopDelay=false
   )
   &SpawnBurst_Instantaneous(Spawn Count=20, Use Spawn Probability=false, UseLoopCountLimit=false)
   /SetVariables(Emitter.CustomTimer:NiagaraFloat=&Add_Float(
     A=$Emitter.CustomTimer
     B=&Multiply_Float(A=$Engine.DeltaTime, B=&Lerp_Float(A=1, B=2, Alpha=$System.ElimAlpha))
    )) {isSetParameters=true}
  ]
  particleSpawnScript [
   &InitializeParticle(
    Color Mode=::NewEnumerator0
    Lifetime=5
    Mesh Scale Mode=::NewEnumerator3
    Mesh Uniform Scale=0.5
    Sprite Size=<10 20>
    Sprite Size Mode=::NewEnumerator4
    UsePositionOffset=false
    Write Material Random=true
   )
   /SetVariables(Particles.TopBot:NiagaraBool=&SetBoolByIntComparison(A=&ReturnExecIndex(), B=4, Comparison Type=::NewEnumerator6)) {isSetParameters=true}
   /SetVariables(Particles.Position=&AddVectorToPosition(
     Position=&SelectVectorFromArray(
      Direct Array Index=&ReturnExecIndex()
      Vector Selection Array=DataInterface() {propertyValues {FloatData [
         <11.7 -2.1 5.5>
         <-11.7 -2.1 5.5>
         <11.7 -25.5 5.5>
         <-11.7 -25.6 5.5>
         <2.9 -2.5 -16.6>
         <-2.9 -2.5 -16.6>
         <8.25 -2.5 -16.6>
         <-8.25 -2.5 -16.6>
         <2.9 -25.5 -16.6>
         <-2.9 -25.5 -16.6>
         <8.25 -25.5 -16.6>
         <-8.25 -25.5 -16.6>
         <11.4 -5.7 -16.6>
         <11.4 -11.2 -16.6>
         <11.4 -16.75 -16.6>
         <11.4 -22.1 -16.6>
         <-11.4 -5.7 -16.6>
         <-11.4 -11.2 -16.6>
         <-11.4 -16.75 -16.6>
         <-11.4 -22.1 -16.6>
        ]}}
     )
     Vector=<0 -1.3 0>
    )) {isSetParameters=true}
   /SetVariables(Particles.SpotSubUV=2) {isSetParameters=true}
  ]
  particleUpdateScript [
   &ParticleState(Kill Particles When Lifetime Has Elapsed=false)
   &ScaleSpriteSize(
    Scale Factor=&MakeVector2D(
     X=&Lerp_Float(A=0.75, B=1, Alpha=$System.ElimAlpha)
     Y=&Multiply_Float(A=$System.FadeIn, B=&Lerp_Float(A=0.75, B=1.5, Alpha=$System.ElimAlpha))
    )
    Scale Sprite Size Mode=::NewEnumerator2
   )
   /Niagara/Modules/Update/Utility/VectorNoiseValue(
    Coordinate Space:ENiagaraCoordinateSpace=::Local
    Noise Frequency=75
    Noise Quality / Cost:ENiagaraVectorNoiseQuality=::NewEnumerator0
    Pan Noise=true
    Sample Position:NiagaraPosition=&Add_Vector(
     A=Linked() {linkedVariable {name=Particles.Position, type=NiagaraPosition}}
     B=&Multiply_VectorByFloat(Vector=<0 0 1>, Float=&Multiply_Float(A=$Emitter.CustomTimer, B=10))
    )
   )
   &InterpolateOverTime(Rate Of Change=0.75, Target Vector Value:Vector3f=$Output.VectorNoiseValue.Noise, Variable Type=::NewEnumerator2)
   /SetVariables(Particles.ElimAlignment:Vector3f=&Add_Vector(
     A=&Subtract_Position(
      A=&AddVectorToPosition(
       Position=$Particles.Position
       Vector=&Add_Vector(
        A=&MakeCustomVectorFromBool(Bool=$Particles.TopBot, True Vector=<0 0 20>, False Vector=<0 0 -20>)
        B=&Multiply_VectorByFloat(
         Vector=&MakeVector(
          X=&Sine(
           Normalized Angle=&Add_Float(
            A=$Emitter.CustomTimer
            B=&Multiply_FloatByInt(Float=0.25, Integer=&ReturnExecIndex())
           )
           Period=2
           Scale=1
           Bias=0
          )
          Y=&Cosine(
           Normalized Angle=&Add_Float(
            A=$Emitter.CustomTimer
            B=&Multiply_FloatByInt(Float=0.25, Integer=&ReturnExecIndex())
           )
           Period=2
           Scale=1
           Bias=0
          )
          Z=0
         )
         Float=&Lerp_Float(A=10, B=20, Alpha=$System.ElimAlpha)
        )
       )
      )
      B=$Particles.Position
     )
     B=&Multiply_VectorByFloat(
      Vector=&Subtract_Vector(
       A=Linked() {linkedVariable {name=Particles.Position, type=NiagaraPosition}}
       B=&MakeCustomVectorFromBool(Bool=$Particles.TopBot, True Vector=<0 -17 0>, False Vector=<0 -17 -15>)
      )
      Float=1.5
     )
    )) {isSetParameters=true}
   &SpriteFacingAndAlignment(Sprite Alignment=$Particles.ElimAlignment, WriteAlignment=true, WriteFacing=false)
   &Color(Color=&MakeLinearColorFromVectorAndFloat(
     "Vector (RGB)"=&Multiply_VectorByFloat(
      Vector=&MakeVectorFromLinearColorRGB(Color=&Multiply_LinearColorByFloat(
        Linear Color=&LerpLinearColors(
         StartColor=<rgba 0.2604 0.2573 0.1682 1>
         EndColor=&SelectLinearColorFromArray(
          Array Sampling Mode=::NewEnumerator1
          Direct Array Index=&ReturnExecIndex()
          Direct Array Mode=::NewEnumerator1
          Color Selection Array=DataInterface() {propertyValues {ColorData [<rgba 0 0.3191 1 1>, <rgba 0.7 0.7 0.7 1>, <rgba 1 0 0 1>]}}
          Random Seed=?
         )
         LerpFactor=$System.ElimAlpha
        )
        Float=&Lerp_Float(A=1, B=2, Alpha=$System.ElimAlpha)
       ))
      Float=&Add_Float(
       A=1
       B=&Waveform(
        Global Phase=&Add_Float(
         A=$Emitter.CustomTimer
         B=&Multiply_FloatByInt(Float=0.1, Integer=&ReturnExecIndex())
        )
        "Waveform [1]"=::NewEnumerator0
        "[1] Frequency"=1
        "[1] Amplitude Min/Max"=<0 1>
        "[1] Scale"=&Lerp_Float(A=0.5, B=1.5, Alpha=$System.ElimAlpha)
       )
      )
     )
     "Float (Alpha)"=$System.FadeIn
    ))
   &FN_Sprite_Attributes(
    Camera Offset=2
    Create Camera Offset Attribute=true
    Create SpriteSize Attribute=true
    SpriteSize=&Vector2DFromFloat(Value=&Lerp_Float(A=7, B=7, Alpha=$System.ElimAlpha))
    Write to Module Namespace=true
   )
   &FN_Sprite_Attributes(
    Camera Offset=3
    Color=&MultiplyLinearColor(Linear Color=$Particles.Color, Scale Factor=<1 1 1 0.1>)
    Create Camera Offset Attribute=true
    Create Color Attribute=true
    Create SpriteSize Attribute=true
    SpriteSize=&Multiply_Vector2D_ByFloat(Float=2, Vector2D=$Particles.FN_Sprite_Attributes.SpriteSize)
    Write to Module Namespace=true
   ) {module=FN_Sprite_Attributes001}
   &FN_Sprite_Attributes(
    Camera Offset=3
    Color=&MultiplyLinearColor(Linear Color=$Particles.Color, Scale Factor=<1 1 1 3>)
    Create Camera Offset Attribute=true
    Create Color Attribute=true
    Create SpriteSize Attribute=true
    SpriteSize=&Multiply_Vector2D(A=<0.3 0.3>, B=$Particles.SpriteSize)
    Write to Module Namespace=true
   ) {module=FN_Sprite_Attributes002}
  ]
  rendererClasses [&NiagaraSpriteRendererProperties]
  renderers [
   {rendererClass=&NiagaraSpriteRendererProperties}
   {rendererClass=&NiagaraSpriteRendererProperties, rendererIndex=1}
   {rendererClass=&NiagaraSpriteRendererProperties, rendererIndex=2}
   {rendererClass=&NiagaraSpriteRendererProperties, rendererIndex=3}
  ]
 }}
summary {
 systemName=NS_OatMealSpreadGolem
 emitters [{emitterName=NE_Spotlights, rendererClasses [&NiagaraSpriteRendererProperties]}]
 userVariables [User.KillCount {type=NiagaraInt32}]
}
