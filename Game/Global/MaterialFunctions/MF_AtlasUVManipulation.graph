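# Material function that offsets and scales UVs within a rectangular sub-region, named for use with texture atlases.
# It finds the midpoint of UV Start and UV Stop, then divides the UVs' distance from it by Texture Scale to scale around that center.
# Texture Offset, also divided by scale, is added, and the result is clamped to the Start and Stop bounds so sampling stays inside the region.
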MaterialFunction /Game/Global/MaterialFunctions/MF_AtlasUVManipulation
graph {nodes [
  n1 = FunctionInput() {bUsePreviewValueAsDefault=true, inputName=Texture Scale, inputType=FunctionInput_Vector2, sortPriority=2}
  n2 = Divide(A=FunctionInput() {bUsePreviewValueAsDefault=true, inputName=Texture Offset, inputType=FunctionInput_Vector2, sortPriority=1}, B=@n1)
  n3 = FunctionInput() {bUsePreviewValueAsDefault=true, inputName=UVs, inputType=FunctionInput_Vector2}
  n4 = FunctionInput() {bUsePreviewValueAsDefault=true, inputName=UV Start, inputType=FunctionInput_Vector2, sortPriority=3}
  n5 = FunctionInput() {bUsePreviewValueAsDefault=true, inputName=UV Stop, inputType=FunctionInput_Vector2, sortPriority=4}
  n6 = Multiply(A=Add(A=@n4, B=@n5), B=Constant2Vector() {g=0.5, r=0.5})
  n7 = Add(A=Divide(A=Subtract(A=@n3, B=@n6), B=@n1), B=@n6)
  FunctionOutput(None=Clamp(None=Add(A=@n2, B=@n7), Min=@n4, Max=@n5))
 ]}
