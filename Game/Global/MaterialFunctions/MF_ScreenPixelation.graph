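# Material function that gives a surface a pixelated screen look, with an RGB sub-pixel pattern that shows up close.
# It snaps input UVs to a Screen Resolution grid with multiply, floor and divide, output as PixelateUV.
# The unsnapped scaled UVs tile the Screen Pixel Pattern texture once per pixel, brightened for the RGB Screen Pixels output.
# That pattern fades linearly to zero with camera distance, over a Fade Distance scaled from meters to centimeters.
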MaterialFunction /Game/Global/MaterialFunctions/MF_ScreenPixelation
graph {nodes [
  n1 = FunctionInput(Preview=TextureCoordinate()) {bUsePreviewValueAsDefault=true, inputName=UV, inputType=FunctionInput_Vector2, previewValue=<1 1 0 1>}
  n2 = FunctionInput() {bUsePreviewValueAsDefault=true, inputName=Screen Resolution, inputType=FunctionInput_Vector2, sortPriority=2, previewValue=<100 100 0 1>}
  n3 = Multiply(A=@n1, B=@n2)
  FunctionOutput(None=Divide(A=Floor(None=@n3), B=@n2)) {outputName=PixelateUV}
  n4 = FunctionInput(Preview=TextureObject() {texture=&T_WhitesRGB}) {
   bUsePreviewValueAsDefault=true
   inputName=Screen Pixel Pattern
   inputType=FunctionInput_Texture2D
   sortPriority=1
   previewValue=<100 100 0 1>
  }
  n5 = TextureSample(UVs=@n3, Tex=@n4) {samplerSource=SSM_Wrap_WorldGroupSettings}
  n6 = Subtract(A=CameraPositionWS(), B=WorldPosition().XYZ)
  n7 = FunctionInput() {bUsePreviewValueAsDefault=true, inputName=Fade Distance, inputType=FunctionInput_Scalar, sortPriority=3, previewValue=<1 1 0 1>}
  n8 = Saturate(None=Divide(A=Length(None=@n6) {is=class MaterialExpressionLength}, B=Multiply(A=@n7) {constB=100}))
  FunctionOutput(None=Multiply(A=Multiply(A=@n5.RGBA) {constB=5}, B=OneMinus(None=@n8))) {outputName=RGB Screen Pixels, sortPriority=1}
 ]}
