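# Material function for the OatmealSpreadGolem back bling that adds an animated, pixelated basketball scoreboard screen to incoming material attributes via masks.
# Time since Blueprint-set GameTimeFromBP samples curve-atlas rows that drive flipbook frames, motion, fades and transitions.
# Atlas layers for graphics, ball, two Score digits, symbols and panning banner composite over a custom HLSL dot background into gamma-adjusted emissive.
# Explicit derivatives and powered alphas address Android mip artifacts.
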MaterialFunction /BRCosmetics/Accessories/FORT_Backpacks/Backpack_M_MED_OatmealSpreadGolem/Materials/MF_OatmealSpreadGolem_ReactiveScreen
graph {nodes [
  n1 = GetMaterialAttributes(None=FunctionInput() {inputType=FunctionInput_MaterialAttributes}) {attributeGetTypes [
    #1
    #2
    #3
    #4
    #5
    #6
   ]}
  n2 = NamedRerouteUsage()
  n3 = TextureSample(UVs=NamedRerouteUsage()) {samplerSource=SSM_Wrap_WorldGroupSettings, texture=/Game/Global/Textures/Default/Noise/Noise}
  n4 = ComponentMask(None=FunctionInput() {inputName=IN_TAMask, sortPriority=1}) {g=true}
  n5 = Multiply(A=@n1.EmissiveColor, B=@n4)
  n6 = Multiply(A=@n4, B=Constant3Vector() {constant=<rgba 1 1 0 0>})
  n7 = Multiply(A=@n6, B=@n6)
  n8 = NamedRerouteUsage()
  n9 = SetMaterialAttributes!(
   MaterialAttributes=@n1.MaterialAttributes
   BaseColor=LinearInterpolate(A=@n1.BaseColor, B=Constant(), Alpha=@n2) {constB=0}
   Metallic=LinearInterpolate(A=@n1.Metallic, B=Constant(), Alpha=@n2)
   Normal=@n1.MaterialAttributes
   Roughness=LinearInterpolate(A=@n1.Roughness, B=LinearInterpolate(Alpha=@n3.R) {constA=0.35, constB=0.88}, Alpha=@n2)
   EmissiveColor=LinearInterpolate(A=LinearInterpolate(A=@n5, B=@n7, Alpha=@n8), B=NamedRerouteUsage(), Alpha=@n2)
   AmbientOcclusion=LinearInterpolate(A=@n1.AmbientOcclusion, Alpha=@n2)
  ) {attributeSetTypes [
    #1
    #2
    #3
    #4
    #5
    #6
   ]}
  FunctionOutput(None=@n9)
  n10 = VectorParameter() {
   group=ReactiveScreen
   parameterName=D Texture Manipulator
   sortPriority=9
   channelNames {a=ScaleY, b=ScaleX, g=OffsetY, r=OffsetX}
   defaultValue=<rgba 0 -0.1 1 1>
  }
  n11 = VectorParameter() {group=ReactiveScreen, parameterName=D Texture UV Coordinates, sortPriority=9, channelNames {a=Buttom, b=Right, g=Top, r=Left}}
  n12 = NamedRerouteUsage()
  n13 = TextureSample(
   UVs=MaterialFunctionCall!(
    UVs=NamedRerouteUsage()
    Texture Offset=MaterialFunctionCall(X=@n10.OffsetX, Y=@n10.OffsetX) {materialFunction=&MakeFloat2}.Result
    Texture Scale=MaterialFunctionCall(X=@n10.ScaleX, Y=@n10.ScaleX) {materialFunction=&MakeFloat2}.Result
    UV Start=MaterialFunctionCall(X=@n11.Left, Y=@n11.Left) {materialFunction=&MakeFloat2}.Result
    UV Stop=MaterialFunctionCall(X=@n11.Right, Y=@n11.Right) {materialFunction=&MakeFloat2}.Result
   ) {materialFunction=/Game/Global/MaterialFunctions/MF_AtlasUVManipulation}.Result
   Tex=@n12
  ) {samplerSource=SSM_Wrap_WorldGroupSettings}
  LinearInterpolate(B=@n13.RGB, Alpha=Multiply(A=@n13.A) {constB=0})
  n14 = VectorParameter() {group=ReactiveScreen, parameterName=ABCDTextureOpacity, sortPriority=5, defaultValue=<rgba 1 1 1 1>}
  MaterialFunctionCall(Float4=MaterialFunctionCall(R=@n14.R, G=@n14.R, B=@n14.R, A=@n14.R) {materialFunction=&AppendMany}.RGBA) {materialFunction=&BreakOutFloat4Components}
  n15 = VertexColor()
  n16 = Add(A=@n15.G, B=@n15.G)
  NamedRerouteDeclaration(None=Step(X=Subtract(A=@n15.R, B=@n16)) {constX=0.5, constY=0.01}) {name=ScreenMask, nodeColor=<rgba 0.8984 0 1 1>}
  MaterialExpressionNamedRerouteUsage_10 = NamedRerouteUsage()
  MaterialExpressionMultiply_31 = Multiply(A=ScalarParameter() {defaultValue=1, group=ReactiveScreen, parameterName=MainScreen_DotSpeed}) {constB=1024}
  n17 = NamedRerouteUsage()
  n18 = Constant2Vector() {g=256, r=256}
  MaterialExpressionMultiply_19 = Multiply(A=@n17, B=@n18) {constB=100}
  MaterialExpressionScalarParameter_7 = ScalarParameter() {defaultValue=5, group=ReactiveScreen, parameterName=MainScreen_DotSize}
  MaterialExpressionConstant3Vector_0 = Constant3Vector() {constant=<rgba 1 0.0196 0.0078 1>}
  MaterialExpressionConstant3Vector_3 = Constant3Vector() {constant=<rgba 0.7176 0.0118 0.0314 1>}
  MaterialExpressionConstant3Vector_4 = Constant3Vector() {constant=<rgba 0.0015 0.0762 0.2874 1>}
  n19 = VectorParameter() {
   group=ReactiveScreen
   parameterName=MainScreen_BG_MotionVector
   channelNames {a="", b="", g=Vector.Y, r=Vector.X}
   defaultValue=<rgba 1 0 0 1>
  }
  MaterialExpressionDivide_6 = Divide(A=ComponentMask(None=@n19) {g=true, r=true}, B=@n18) {constB=10}
  MaterialExpressionScalarParameter_14 = ScalarParameter() {defaultValue=0.01, group=ReactiveScreen, parameterName=MainScreen_DotHardness}
  n20 = VectorParameter() {group=ReactiveScreen, parameterName=MainScreen_BG_Canvas, channelNames {a=Size, b=Size, g=Offset, r=Offset}, defaultValue=<rgba 0 0 2 2>}
  n21 = MaterialFunctionCall(X=@n20.Offset, Y=@n20.Offset) {materialFunction=&MakeFloat2}
  n22 = MaterialFunctionCall(Vector 2=Multiply(
    A=Subtract(A=@n17, B=@n21.Result)
    B=MaterialFunctionCall(X=@n20.Size, Y=@n20.Size) {materialFunction=&MakeFloat2}.Result
   )) {materialFunction=/Engine/Functions/Engine_MaterialFunctions02/Utility/VectorLength}
  n23 = ScalarParameter() {defaultValue=1, group=ReactiveScreen, parameterName=MainScreen_DotGraphSize}
  MaterialExpressionSubtract_1 = Subtract(A=@n22.V2 Length, B=Subtract(A=@n23) {constB=0.5})
  MaterialExpressionSubtract_8 = Subtract(A=@n22.V2 Length, B=Add(A=@n23) {constB=0.5})
  n24 = Custom!(
   Time=@MaterialExpressionNamedRerouteUsage_10
   Speed=@MaterialExpressionMultiply_31
   UVCoords=@MaterialExpressionMultiply_19
   size=@MaterialExpressionScalarParameter_7
   c1=@MaterialExpressionConstant3Vector_0
   c2=@MaterialExpressionConstant3Vector_3
   c3=@MaterialExpressionConstant3Vector_4
   edgeVec=@MaterialExpressionDivide_6
   edge=@MaterialExpressionScalarParameter_14
   pattern1=@MaterialExpressionSubtract_1
   pattern2=@MaterialExpressionSubtract_8
  ) {
   code="// The below expression will get compiled\r\n// into the output of this node\r\n\r\nfloat t = Time * Speed;\r\nfloat2 animPos = UVCoords - edgeVec*t;\r\nfloat2 d = frac(animPos/size)-0.5;\r\n\r\nfloat s = dot(UVCoords,edgeVec);\r\nfloat h1 = smoothstep(0.5 - edge, 0.5+edge, length(d)+pattern1);\r\nfloat h2 = smoothstep(0.5 - edge, 0.5+edge, length(d)+pattern2);\r\n\r\n\r\n//float3 bgc1 = c1*h1 + (1.0-h1)*c2;\r\nfloat3 bg1 = lerp(c1, c2,h1);\r\nfloat3 bg2 = lerp(bg1,c3,h2);\r\n\r\n\r\n\r\nreturn bg2;"
   description=HFT
   inputs [
    {
     inputName=Time
     input {expression=&self:MaterialExpressionNamedRerouteUsage_10}
    }
    {
     inputName=Speed
     input {expression=&self:MaterialExpressionMultiply_31}
    }
    {
     inputName=UVCoords
     input {expression=&self:MaterialExpressionMultiply_19}
    }
    {
     inputName=size
     input {expression=&self:MaterialExpressionScalarParameter_7}
    }
    {
     inputName=c1
     input {
      expression=&self:MaterialExpressionConstant3Vector_0
      mask=1
      maskB=1
      maskG=1
      maskR=1
     }
    }
    {
     inputName=c2
     input {
      expression=&self:MaterialExpressionConstant3Vector_3
      mask=1
      maskB=1
      maskG=1
      maskR=1
     }
    }
    {
     inputName=c3
     input {
      expression=&self:MaterialExpressionConstant3Vector_4
      mask=1
      maskB=1
      maskG=1
      maskR=1
     }
    }
    {
     inputName=edgeVec
     input {expression=&self:MaterialExpressionDivide_6}
    }
    {
     inputName=edge
     input {expression=&self:MaterialExpressionScalarParameter_14}
    }
    {
     inputName=pattern1
     input {expression=&self:MaterialExpressionSubtract_1}
    }
    {
     inputName=pattern2
     input {expression=&self:MaterialExpressionSubtract_8}
    }
   ]
  }
  n25 = Constant2Vector() {g=0.05}
  n26 = Add(A=TextureCoordinate() {coordinateIndex=1, desc=UV2}, B=@n25)
  n27 = ComponentMask(None=@n26) {g=true}
  n28 = OneMinus(None=@n27)
  n29 = Step(Y=OneMinus(None=Max(A=@n28, B=@n27)), X=OneMinus(None=Multiply(A=@n25.G) {constB=2})) {constX=0.1, constY=0.1}
  n30 = Multiply(A=@n29, B=Constant3Vector() {constant=<rgba 0.9725 0.7451 0.0353 1>})
  n31 = Multiply(A=@n29, B=Constant3Vector() {constant=<rgba 1 0.6157 0.0157 1>})
  n32 = Step(Y=ComponentMask(None=@n26) {g=true}) {constX=0.5, constY=0.1}
  NamedRerouteDeclaration(None=LinearInterpolate(A=@n24, B=LinearInterpolate(A=@n30, B=@n31, Alpha=@n32), Alpha=NamedRerouteUsage())) {name=Background, nodeColor=<rgba 0 1 0.8125 1>}
  n33 = TextureCoordinate() {coordinateIndex=1, desc=UV2}
  NamedRerouteDeclaration(None=Step(X=ComponentMask(None=@n33) {g=true}) {constX=0.7, constY=0.2}) {name=BannerScreenMask, nodeColor=<rgba 0 0.9062 1 1>}
  n34 = Time()
  n35 = Subtract(A=@n34, B=ScalarParameter() {defaultValue=-99999, group=ReactiveScreen, parameterName=GameTimeFromBP, sortPriority=100})
  n36 = ScalarParameter() {defaultValue=2, group=ReactiveScreen, parameterName=Animation_Duration, sortPriority=101}
  NamedRerouteDeclaration(None=Saturate(None=Divide(A=@n35, B=@n36))) {name=TimeNative, nodeColor=<rgba 0.2891 1 0 1>}
  n37 = TextureObjectParameter() {group=ReactiveScreen, parameterName=ScreenPixelPattern, sortPriority=0, texture=&T_WhitesRGB}
  n38 = NamedRerouteUsage()
  n39 = MaterialFunctionCall!(
   UV=NamedRerouteUsage()
   Screen Pixel Pattern=@n37
   Screen Resolution=LinearInterpolate(A=Multiply(A=@n38) {constB=2}, B=@n38, Alpha=NamedRerouteUsage()) {constA=1.1}
   Fade Distance=ScalarParameter() {defaultValue=40, group=ReactiveScreen, parameterName=Screen Fade Distance, sortPriority=13}
  ) {materialFunction=/Game/Global/MaterialFunctions/MF_ScreenPixelation}
  NamedRerouteDeclaration(None=MaterialFunctionCall(UVs=@n39.PixelateUV, Texture Scale=NamedRerouteUsage()) {materialFunction=&ScaleUVsByCenter}.UVs) {name=LayerFlipbook_UV, nodeColor=<rgba 0 1 0.8828 1>}
  NamedRerouteDeclaration(None=@n33) {name=ScreenUV, nodeColor=<rgba 0.4297 1 0 1>}
  n40 = NamedRerouteUsage()
  n41 = CurveAtlasRowParameter(CurveTime=@n40) {
   atlas=/BRCosmetics/Accessories/FORT_Backpacks/Backpack_M_MED_OatmealSpreadGolem/Textures/OatmealSpreadGolem_CurveAtlas
   curve=/BRCosmetics/Accessories/FORT_Backpacks/Backpack_M_MED_OatmealSpreadGolem/Textures/OatmealSpreadGolem_FlipbookCurve
   defaultValue=1
   group=ReactiveScreen
   parameterName=FlipbookCurve
   sortPriority=1
  }
  n42 = Modulo!(A=@n41.A, B=Constant() {r=4})
  NamedRerouteDeclaration(None=Divide(A=Floor(None=@n42)) {constB=4}) {name=FlipbookFrames, nodeColor=<rgba 0.0781 0 1 1>}
  n43 = CurveAtlasRowParameter(CurveTime=@n40) {
   atlas=&OatmealSpreadGolem_CurveAtlas
   curve=/BRCosmetics/Accessories/FORT_Backpacks/Backpack_M_MED_OatmealSpreadGolem/Textures/OatmealSpreadGolem_AlphaCurve
   defaultValue=3
   group=ReactiveScreen
   parameterName=AlphaCurve
   sortPriority=1
  }
  NamedRerouteDeclaration(None=@n43.A) {name=PopAlpha, nodeColor=<rgba 1 0 0.7969 1>}
  n44 = CurveAtlasRowParameter(CurveTime=@n40) {
   atlas=&OatmealSpreadGolem_CurveAtlas
   curve=/BRCosmetics/Accessories/FORT_Backpacks/Backpack_M_MED_OatmealSpreadGolem/Textures/OatmealSpreadGolem_TransitionCurve
   defaultValue=2
   group=ReactiveScreen
   parameterName=TransitionCurve
   sortPriority=1
  }
  NamedRerouteDeclaration(None=@n44.A) {name=BGTransition, nodeColor=<rgba 1 0 0.3516 1>}
  NamedRerouteDeclaration(None=@n44.B) {name=Blink, nodeColor=<rgba 1 0 0.1406 1>}
  NamedRerouteDeclaration(None=@n41.R) {name=FlipbookMap.x, nodeColor=<rgba 0.1484 1 0 1>}
  NamedRerouteDeclaration(None=@n41.G) {name=FlipbookMap.y, nodeColor=<rgba 1 0 0.4687 1>}
  n45 = VectorParameter() {
   group=ReactiveScreen
   parameterName=ScreenResolution
   sortPriority=12
   channelNames {a="", b="", g=Resolution Y, r=Resolution X}
   defaultValue=<rgba 40 40 0 0>
  }
  NamedRerouteDeclaration(None=MaterialFunctionCall(X=@n45.Resolution X, Y=@n45.Resolution X) {materialFunction=&MakeFloat2}.Result) {name=ScreenResolution, nodeColor=<rgba 0.9453 0 1 1>}
  NamedRerouteDeclaration(None=@n41.B) {name=FlipbookMap.size, nodeColor=<rgba 0 0.7891 1 1>}
  n46 = CurveAtlasRowParameter(CurveTime=@n40) {
   atlas=&OatmealSpreadGolem_CurveAtlas
   curve=/BRCosmetics/Accessories/FORT_Backpacks/Backpack_M_MED_OatmealSpreadGolem/Textures/OatmealSpreadGolem_BasketballCurve
   group=ReactiveScreen
   parameterName=BasketballCurve
   sortPriority=1
  }
  NamedRerouteDeclaration(None=@n46.R) {name=Ball.x, nodeColor=<rgba 1 0.8906 0 1>}
  NamedRerouteDeclaration(None=Subtract(A=@n46.G)) {name=Ball.y, nodeColor=<rgba 0.5703 0 1 1>}
  NamedRerouteDeclaration(None=@n46.B) {name=Ball.size, nodeColor=<rgba 0 1 0.25 1>}
  NamedRerouteDeclaration(None=@n43.G) {name=BallAlpha, nodeColor=<rgba 1 0 0.0703 1>}
  NamedRerouteDeclaration(None=@n44.R) {name=TextTransition, nodeColor=<rgba 0 0.3906 1 1>}
  NamedRerouteDeclaration(None=@n44.G) {name=FlipbookMapTransition, nodeColor=<rgba 0.7109 1 0 1>}
  n47 = Modulo(A=@n46.A, B=Constant() {r=12})
  NamedRerouteDeclaration(None=Divide(A=Floor(None=@n47)) {constB=12}) {name=BallFrames, nodeColor=<rgba 1 0.9141 0 1>}
  NamedRerouteDeclaration(None=@n39.PixelateUV) {name=Pixelate UV, nodeColor=<rgba 1 0.7031 0 1>}
  n48 = TextureObjectParameter() {
   group=ReactiveScreen
   parameterName=Screen Texture
   sortPriority=0
   texture=/Engine/Functions/Engine_MaterialFunctions02/ExampleContent/Textures/flipbook
  }
  NamedRerouteDeclaration(None=@n48) {name=AtlasTexture, nodeColor=<rgba 0.0042 0.2727 1 1>}
  NamedRerouteDeclaration(None=MaterialFunctionCall(UVs=@n39.PixelateUV, Texture Scale=NamedRerouteUsage()) {materialFunction=&ScaleUVsByCenter}.UVs) {name=LayerText_UV, nodeColor=<rgba 0.3359 1 0 1>}
  n49 = VectorParameter() {
   group=ReactiveScreen
   parameterName=C Texture Manipulator
   sortPriority=8
   channelNames {a=ScaleY, b=ScaleX, g=OffsetY, r=OffsetX}
   defaultValue=<rgba 0 0 1 1>
  }
  n50 = VectorParameter() {group=ReactiveScreen, parameterName=C Texture UV Coordinates, sortPriority=8, channelNames {a=Buttom, b=Right, g=Top, r=Left}}
  n51 = TextureCoordinate() {coordinateIndex=1}
  n52 = Divide(A=@n51, B=NamedRerouteUsage())
  n53 = TextureSample!(
   UVs=MaterialFunctionCall(
    UVs=NamedRerouteUsage()
    Texture Offset=MaterialFunctionCall(X=@n49.OffsetX, Y=@n49.OffsetX) {materialFunction=&MakeFloat2}.Result
    Texture Scale=MaterialFunctionCall(X=@n49.ScaleX, Y=@n49.ScaleX) {materialFunction=&MakeFloat2}.Result
    UV Start=MaterialFunctionCall(X=@n50.Left, Y=@n50.Left) {materialFunction=&MakeFloat2}.Result
    UV Stop=MaterialFunctionCall(X=@n50.Right, Y=@n50.Right) {materialFunction=&MakeFloat2}.Result
   ) {materialFunction=&MF_AtlasUVManipulation}.Result
   Tex=@n12
   "DDX(UVs)"=DDX(Value=@n52)
   "DDY(UVs)"=DDY(Value=@n52)
   Apply View MipBias
  ) {mipValueMode=TMVM_Derivative, samplerSource=SSM_Wrap_WorldGroupSettings}
  n54 = Power(Base=@n53.A) {constExponent=5, desc=Fix alpha mip artifact on Android}
  n55 = NamedRerouteUsage()
  NamedRerouteDeclaration(None=LinearInterpolate(A=NamedRerouteUsage() {desc=BG Layer}, B=@n53.RGB, Alpha=Multiply(A=@n54, B=@n55)) {desc=Layer 1 - Pop graph}) {name=Layer_01_Pop, nodeColor=<rgba 0 0.9062 1 1>}
  n56 = VectorParameter() {group=ReactiveScreen, parameterName=GraphFlipbookSize, defaultValue=<rgba 4 4 0 1>}
  n57 = NamedRerouteUsage()
  n58 = MaterialFunctionCall(X=NamedRerouteUsage(), Y=NamedRerouteUsage()) {materialFunction=&MakeFloat2}
  n59 = VectorParameter() {
   group=ReactiveScreen
   parameterName=Graph Manipulator
   sortPriority=6
   channelNames {a=ScaleY, b=ScaleX, g=OffsetY, r=OffsetX}
   defaultValue=<rgba 0 0 1 1>
  }
  n60 = Multiply(
   A=MaterialFunctionCall(X=@n59.ScaleX, Y=@n59.ScaleX) {materialFunction=&MakeFloat2}.Result
   B=NamedRerouteUsage()
  )
  n61 = VectorParameter() {
   group=ReactiveScreen
   parameterName=Graph UV Coordinates
   sortPriority=6
   channelNames {a=Buttom, b=Right, g=Top, r=Left}
   defaultValue=<rgba 0 0 1 1>
  }
  n62 = MaterialFunctionCall(X=@n61.Left, Y=@n61.Left) {materialFunction=&MakeFloat2}
  n63 = MaterialFunctionCall(X=@n61.Right, Y=@n61.Right) {materialFunction=&MakeFloat2}
  n64 = Divide(A=TextureCoordinate() {coordinateIndex=1}, B=NamedRerouteUsage())
  n65 = TextureSample(
   UVs=MaterialFunctionCall!(
    Size=@n56.RGB
    Selector=NamedRerouteUsage()
    UV=MaterialFunctionCall(UVs=@n57, Texture Offset=@n58.Result, Texture Scale=@n60, UV Start=@n62.Result, UV Stop=@n63.Result) {materialFunction=&MF_AtlasUVManipulation}.Result
   ) {materialFunction=/Game/Global/MaterialFunctions/MF_FlipbookUV}.FlipbookUV1
   Tex=NamedRerouteUsage()
   "DDX(UVs)"=DDX(Value=@n64)
   "DDY(UVs)"=DDY(Value=@n64)
  ) {mipValueMode=TMVM_Derivative, samplerSource=SSM_Wrap_WorldGroupSettings}
  n66 = Power(Base=@n65.A) {constExponent=5, desc=Fix alpha mip artifact on Android}
  n67 = NamedRerouteUsage()
  NamedRerouteDeclaration(None=LinearInterpolate(A=NamedRerouteUsage(), B=@n65.RGB, Alpha=Multiply(A=@n66, B=@n67)) {desc=Layer 2 ball}) {name=Layer_02_Ball, nodeColor=<rgba 1 0.7734 0 1>}
  n68 = VectorParameter() {group=ReactiveScreen, parameterName=BallFlipbookSize, defaultValue=<rgba 12 12 0 1>}
  n69 = NamedRerouteUsage()
  n70 = MaterialFunctionCall(X=NamedRerouteUsage(), Y=NamedRerouteUsage()) {materialFunction=&MakeFloat2}
  n71 = VectorParameter() {
   group=ReactiveScreen
   parameterName=Ball Manipulator
   sortPriority=6
   channelNames {a=ScaleY, b=ScaleX, g=OffsetY, r=OffsetX}
   defaultValue=<rgba 0 0 1 1>
  }
  n72 = Multiply(
   A=MaterialFunctionCall(X=@n71.ScaleX, Y=@n71.ScaleX) {materialFunction=&MakeFloat2}.Result
   B=NamedRerouteUsage()
  )
  n73 = VectorParameter() {
   group=ReactiveScreen
   parameterName=Ball UV Coordinates
   sortPriority=6
   channelNames {a=Buttom, b=Right, g=Top, r=Left}
   defaultValue=<rgba 0 0 1 1>
  }
  n74 = MaterialFunctionCall(X=@n73.Left, Y=@n73.Left) {materialFunction=&MakeFloat2}
  n75 = MaterialFunctionCall(X=@n73.Right, Y=@n73.Right) {materialFunction=&MakeFloat2}
  n76 = Divide(A=TextureCoordinate() {coordinateIndex=1}, B=NamedRerouteUsage())
  n77 = TextureSample(
   UVs=MaterialFunctionCall(
    Size=@n68.RGB
    Selector=NamedRerouteUsage()
    UV=MaterialFunctionCall(UVs=@n69, Texture Offset=@n70.Result, Texture Scale=@n72, UV Start=@n74.Result, UV Stop=@n75.Result) {materialFunction=&MF_AtlasUVManipulation}.Result
   ) {materialFunction=&MF_FlipbookUV}.FlipbookUV1
   Tex=NamedRerouteUsage()
   "DDX(UVs)"=DDX(Value=@n76)
   "DDY(UVs)"=DDY(Value=@n76)
  ) {mipValueMode=TMVM_Derivative, samplerSource=SSM_Wrap_WorldGroupSettings}
  n78 = Power(Base=@n77.A) {constExponent=5, desc=Fix alpha mip artifact on Android}
  n79 = NamedRerouteUsage()
  NamedRerouteDeclaration(None=LinearInterpolate(A=NamedRerouteUsage(), B=@n77.RGB, Alpha=Multiply(A=@n78, B=@n79)) {desc=Layer 3 Flipbook}) {name=Layer_03_flipbook, nodeColor=<rgba 0.4531 0 1 1>}
  n80 = NamedRerouteUsage()
  n81 = VectorParameter() {group=ReactiveScreen, parameterName=DigitsFlipbookSize, defaultValue=<rgba 12 12 0 1>}
  n82 = ScalarParameter() {group=ReactiveScreen, parameterName=Score, sliderMax=99, sortPriority=0}
  n83 = Floor(None=Modulo(A=@n82, B=Constant() {r=10}))
  n84 = NamedRerouteUsage()
  n85 = VectorParameter() {
   group=ReactiveScreen
   parameterName=Digits Manipulator
   sortPriority=6
   channelNames {a=ScaleY, b=ScaleX, g=OffsetY, r=OffsetX}
   defaultValue=<rgba 0 0 1 1>
  }
  n86 = ScalarParameter() {defaultValue=1, group=ReactiveScreen, parameterName=Digits SpaceBetween}
  n87 = MaterialFunctionCall(X=Subtract(A=@n85.OffsetX, B=@n86), Y=@n85.OffsetY) {materialFunction=&MakeFloat2}
  n88 = MaterialFunctionCall(X=@n85.ScaleX, Y=@n85.ScaleX) {materialFunction=&MakeFloat2}
  n89 = VectorParameter() {
   group=ReactiveScreen
   parameterName=Digits UV Coordinates
   sortPriority=6
   channelNames {a=Buttom, b=Right, g=Top, r=Left}
   defaultValue=<rgba 0 0 1 1>
  }
  n90 = MaterialFunctionCall(X=@n89.Left, Y=@n89.Left) {materialFunction=&MakeFloat2}
  n91 = MaterialFunctionCall(X=@n89.Right, Y=@n89.Right) {materialFunction=&MakeFloat2}
  n92 = Divide(A=TextureCoordinate() {coordinateIndex=1}, B=NamedRerouteUsage())
  n93 = DDX(Value=@n92)
  n94 = DDY(Value=@n92)
  n95 = TextureSample(
   UVs=MaterialFunctionCall(
    Size=@n81.RGB
    Selector=Divide(A=@n83) {constB=12}
    UV=MaterialFunctionCall(UVs=@n84, Texture Offset=@n87.Result, Texture Scale=@n88.Result, UV Start=@n90.Result, UV Stop=@n91.Result) {materialFunction=&MF_AtlasUVManipulation}.Result
   ) {materialFunction=&MF_FlipbookUV}.FlipbookUV1
   Tex=NamedRerouteUsage()
   "DDX(UVs)"=@n93
   "DDY(UVs)"=@n94
  ) {mipValueMode=TMVM_Derivative, samplerSource=SSM_Wrap_WorldGroupSettings}
  n96 = Power(Base=@n95.A) {constExponent=5, desc=Fix alpha mip artifact on Android}
  n97 = NamedRerouteUsage()
  n98 = Multiply(A=@n96, B=@n97)
  n99 = Floor(None=Modulo(A=Divide(A=@n82) {constB=10}, B=Constant() {r=10}))
  n100 = NamedRerouteUsage()
  n101 = MaterialFunctionCall(X=Add(A=@n85.OffsetX, B=@n86), Y=@n85.OffsetY) {materialFunction=&MakeFloat2}
  n102 = MaterialFunctionCall(X=@n85.ScaleX, Y=@n85.ScaleX) {materialFunction=&MakeFloat2}
  n103 = MaterialFunctionCall(X=@n89.Left, Y=@n89.Left) {materialFunction=&MakeFloat2}
  n104 = MaterialFunctionCall(X=@n89.Right, Y=@n89.Right) {materialFunction=&MakeFloat2}
  n105 = TextureSample(
   UVs=MaterialFunctionCall(
    Size=@n81.RGB
    Selector=Divide(A=@n99) {constB=12}
    UV=MaterialFunctionCall(UVs=@n100, Texture Offset=@n101.Result, Texture Scale=@n102.Result, UV Start=@n103.Result, UV Stop=@n104.Result) {materialFunction=&MF_AtlasUVManipulation}.Result
   ) {materialFunction=&MF_FlipbookUV}.FlipbookUV1
   Tex=NamedRerouteUsage()
   "DDX(UVs)"=@n93
   "DDY(UVs)"=@n94
  ) {mipValueMode=TMVM_Derivative, samplerSource=SSM_Wrap_WorldGroupSettings}
  n106 = Power(Base=@n105.A) {constExponent=5, desc=Fix alpha mip artifact on Android}
  NamedRerouteDeclaration(None=LinearInterpolate(A=LinearInterpolate(A=@n80, B=@n95.RGB, Alpha=@n98), B=@n105.RGB, Alpha=Multiply(A=@n106, B=@n97))) {name=Layer_04_digits, nodeColor=<rgba 0 1 0.1328 1>}
  n107 = NamedRerouteUsage()
  n108 = VectorParameter() {
   group=ReactiveScreen
   parameterName=A Texture Manipulator
   sortPriority=6
   channelNames {a=ScaleY, b=ScaleX, g=OffsetY, r=OffsetX}
   defaultValue=<rgba 0 0 1 1>
  }
  n109 = VectorParameter() {
   group=ReactiveScreen
   parameterName=A Texture UV Coordinates
   sortPriority=6
   channelNames {a=Buttom, b=Right, g=Top, r=Left}
   defaultValue=<rgba 0 0 1 1>
  }
  n110 = Divide(A=@n51, B=NamedRerouteUsage())
  n111 = DDX(Value=@n110)
  n112 = DDY(Value=@n110)
  n113 = TextureSample(
   UVs=MaterialFunctionCall(
    UVs=NamedRerouteUsage()
    Texture Offset=MaterialFunctionCall(X=@n108.OffsetX, Y=@n108.OffsetX) {materialFunction=&MakeFloat2}.Result
    Texture Scale=MaterialFunctionCall(X=@n108.ScaleX, Y=@n108.ScaleX) {materialFunction=&MakeFloat2}.Result
    UV Start=MaterialFunctionCall(X=@n109.Left, Y=@n109.Left) {materialFunction=&MakeFloat2}.Result
    UV Stop=MaterialFunctionCall(X=@n109.Right, Y=@n109.Right) {materialFunction=&MakeFloat2}.Result
   ) {materialFunction=&MF_AtlasUVManipulation}.Result
   Tex=@n12
   "DDX(UVs)"=@n111
   "DDY(UVs)"=@n112
  ) {mipValueMode=TMVM_Derivative, samplerSource=SSM_Wrap_WorldGroupSettings}
  n114 = Power(Base=@n113.A) {constExponent=5, desc=Fix alpha mip artifact on Android}
  n115 = NamedRerouteUsage()
  n116 = Multiply(A=@n114, B=@n115)
  n117 = VectorParameter() {
   group=ReactiveScreen
   parameterName=B Texture Manipulator
   sortPriority=6
   channelNames {a=ScaleY, b=ScaleX, g=OffsetY, r=OffsetX}
   defaultValue=<rgba 0 0 1 1>
  }
  n118 = VectorParameter() {
   group=ReactiveScreen
   parameterName=B Texture UV Coordinates
   sortPriority=6
   channelNames {a=Buttom, b=Right, g=Top, r=Left}
   defaultValue=<rgba 0 0 0 0.1>
  }
  n119 = TextureSample(
   UVs=MaterialFunctionCall(
    UVs=NamedRerouteUsage()
    Texture Offset=MaterialFunctionCall(X=@n117.OffsetX, Y=@n117.OffsetX) {materialFunction=&MakeFloat2}.Result
    Texture Scale=MaterialFunctionCall(X=@n117.ScaleX, Y=@n117.ScaleX) {materialFunction=&MakeFloat2}.Result
    UV Start=MaterialFunctionCall(X=@n118.Left, Y=@n118.Left) {materialFunction=&MakeFloat2}.Result
    UV Stop=MaterialFunctionCall(X=@n118.Right, Y=@n118.Right) {materialFunction=&MakeFloat2}.Result
   ) {materialFunction=&MF_AtlasUVManipulation}.Result
   Tex=@n12
   "DDX(UVs)"=@n111
   "DDY(UVs)"=@n112
  ) {mipValueMode=TMVM_Derivative, samplerSource=SSM_Wrap_WorldGroupSettings}
  n120 = Power(Base=@n119.A) {constExponent=5, desc=Fix alpha mip artifact on Android}
  NamedRerouteDeclaration(None=LinearInterpolate(A=LinearInterpolate(A=@n107, B=@n113.RGB, Alpha=@n116), B=@n119.RGB, Alpha=Multiply(A=@n120, B=@n115))) {name=Layer_05_symbol, nodeColor=<rgba 1 0 0.1875 1>}
  NamedRerouteDeclaration(None=Add(A=@n114, B=@n120)) {name=SymbolMask, nodeColor=<rgba 0 0.5078 1 1>}
  NamedRerouteDeclaration(None=Add(A=@n96, B=@n106)) {name=DigitsMask, nodeColor=<rgba 0.8281 1 0 1>}
  NamedRerouteDeclaration(None=MaterialFunctionCall(X=@n81.R, Y=@n81.R) {materialFunction=&MakeFloat2}.Result) {name=DigitsSubUV, nodeColor=<rgba 0.1953 0 1 1>}
  NamedRerouteDeclaration(None=MaterialFunctionCall(X=@n56.R, Y=@n56.R) {materialFunction=&MakeFloat2}.Result) {name=GraphSubUV, nodeColor=<rgba 0.0312 0 1 1>}
  NamedRerouteDeclaration(None=MaterialFunctionCall(X=@n68.R, Y=@n68.R) {materialFunction=&MakeFloat2}.Result) {name=BallSubUV, nodeColor=<rgba 0 0.7187 1 1>}
  NamedRerouteDeclaration(None=@n43.R) {name=TextAlpha, nodeColor=<rgba 1 0 0.6328 1>}
  NamedRerouteDeclaration(None=@n39.RGB Screen Pixels) {name=RGB Pixels, nodeColor=<rgba 0 0.1797 1 1>}
  n121 = VectorParameter() {
   group=ReactiveScreen
   parameterName=Bannar Manipulator
   sortPriority=6
   channelNames {a=ScaleY, b=ScaleX, g=OffsetY, r=OffsetX}
   defaultValue=<rgba 0 0 1 1>
  }
  n122 = Multiply(
   A=NamedRerouteUsage()
   B=MaterialFunctionCall(X=@n121.ScaleX, Y=@n121.ScaleX) {materialFunction=&MakeFloat2}.Result
  )
  n123 = MaterialFunctionCall(X=@n121.OffsetX, Y=@n121.OffsetX) {materialFunction=&MakeFloat2}
  n124 = VectorParameter() {
   group=ReactiveScreen
   parameterName=Bannar Panning Speed
   sortPriority=6
   channelNames {a="", b="", g=Speed.Y, r=Speed.X}
   defaultValue=<rgba 0 0 1 1>
  }
  n125 = Panner(
   Coordinate=MaterialFunctionCall(UVs=Add(A=@n122, B=@n123.Result), "Rotation Angle (0-1)"=Constant() {r=-0.25}) {materialFunction=&CustomRotator}.Rotated Values
   Speed=MaterialFunctionCall(X=@n124.Speed.X, Y=@n124.Speed.X) {materialFunction=&MakeFloat2}.Result
  ) {speedY=0.01}
  n126 = NamedRerouteUsage()
  NamedRerouteDeclaration(None=LinearInterpolate(
    A=TextureSample!(UVs=@n125, Tex=@n126, Apply View MipBias) {samplerSource=SSM_Wrap_WorldGroupSettings}.RGB
    B=NamedRerouteUsage()
    Alpha=NamedRerouteUsage()
   )) {name=Layer_06_Bannar, nodeColor=<rgba 1 0 0.6797 1>}
  n127 = NamedRerouteUsage()
  n128 = NamedRerouteUsage()
  n129 = Multiply(A=Multiply(A=Add(A=@n127, B=@n128), B=Constant3Vector() {constant=<rgba 0.8 0.6 0 0>}), B=NamedRerouteUsage())
  n130 = NamedRerouteUsage()
  n131 = Power(
   Base=Add(A=NamedRerouteUsage(), B=Multiply(A=@n129, B=@n130))
   Exp=Max(A=ScalarParameter() {defaultValue=2.2, group=ReactiveScreen, parameterName=Screen Gamma, sortPriority=4}) {constB=0.001}
  )
  n132 = ScalarParameter() {defaultValue=1, group=ReactiveScreen, parameterName=Screen Emissive Intensity, sortPriority=3}
  n133 = NamedRerouteUsage()
  NamedRerouteDeclaration(None=Multiply(A=Multiply(A=@n131, B=@n132), B=ComponentMask(None=@n133) {b=true, g=true, r=true})) {name=FinalGraph, nodeColor=<rgba 0 1 1 1>}
  NamedRerouteDeclaration(None=@n43.B) {name=FlipbookAlpha, nodeColor=<rgba 1 0 0.7969 1>}
  NamedRerouteDeclaration(None=@n34) {name=Time, nodeColor=<rgba 0 0.1797 1 1>}
  NamedRerouteUsage()
 ]}
